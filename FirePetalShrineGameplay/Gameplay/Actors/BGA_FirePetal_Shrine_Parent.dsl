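; Parent Blueprint for the Fire Petal shrine: players carrying a sprite return it to the shrine for a sprite AI and loot, others can reveal nearby sprites on the minimap.
; Interaction grants a player ability and sets replicated ActivationData, whose OnRep drives music, obelisk lift, dummy sprite animation and completion.
; Rewards come from a data registry row keyed by sprite tag.
; A fallback timer re-enables interaction, and hotfix scalable floats tune timings and radar.
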(event EventBeginPlay
  (bind _spritespawner (Variables|Default|GetSpriteSpawner))
  ;[0.2107 0.8438 0.1055] Generate sprites on beginplay
  (Utilities|IsValid _spritespawner
    (:Is%20Valid
      (EventDispatchers|AssignOnActorSpawnedResult _spritespawner (EventDispatchers|CreateEvent :!event Custom|OnSpriteSpawned) :!class /Script/LagerRuntime.FortAthenaLivingWorldVolume)
      (Networking|SwitchHasAuthority
        (:Authority
          (bind _returnvalue (Class|SpriteDiscoverableManager|GetSpriteDiscoverableManager))
          (Class|SpriteDiscoverableManager|RegisterDiscoverable _returnvalue self) ; Register so it can be discovered when holding a sprite
          (CallFunction|SpawnMusicControllerActor))
        (:Remote
          (CallFunction|SpawnAGV))))))

(event Custom|OnInteractRadar
  (params (InteractingPawn :type Fort%20Pawn%20Object%20Reference))
  (bind _returnvalue (Transformation|GetActorLocation))
  (bind _hf_radarspritemarkingradius (Variables|Hotfixes|GetHF_RadarSpriteMarkingRadius))
  (bind _returnvalue_2 (ScalableFloat|GetValueatLevel _hf_radarspritemarkingradius))
  (bind _returnvalue_3 (Fortnite|GetAIDirector))
  (bind _outpawns (AI|GetAIPawnsinRange _returnvalue_3 _returnvalue _returnvalue_2))
  (bind _gc_usedradar_gameplaycuetag (Utilities|Struct|BreakGameplayCueTag (Variables|GetGC_UsedRadar_GameplayCueTag)))
  ;[0.7899 0.1002 0.2833] Sprite Radar interact
  (CallFunction|ServeronInteractRadar (Fortnite|GetFortPlayerControllerfromActor InteractingPawn))
  (CallFunction|AddTeamtoInteractFilter InteractingPawn)
  (for _index (range (Math|Integer|Min(Integer) (Utilities|Array|Length _outpawns) (ScalableFloat|GetValueasInteger (Variables|Hotfixes|GetHF_MaxSpritesToMarkOnRadar))))
    (bind _asnpc_pawn_irwin_simple_sprite (Utilities|Casting|CastToNPC_Pawn_Irwin_Simple_Sprite (Utilities|Array|Get(acopy) _outpawns _index)))
    (Class|NPCPawnIrwinSimpleSprite|RevealSpriteonMinimap _asnpc_pawn_irwin_simple_sprite InteractingPawn)
    (Utilities|Array|Add (Variables|Default|GetMarkedSprites) _asnpc_pawn_irwin_simple_sprite))
  (GameplayCue|ExecuteGameplayCueOnActor(Burst) self _gc_usedradar_gameplaycuetag (Ability|GameplayCue|MakeGameplayCueParameters 0.0))
  (Variables|Default|SetOnInteractCooldown? true)
  (Utilities|FlowControl|Delay (Variables|Default|GetInteractCooldown))
  (Variables|Default|SetOnInteractCooldown? false))

(event Custom|OnActorSpawnedResult_Event
  (params (SpawnedActor :type Actor%20Object%20Reference) (bResult :type Boolean)))

(event Custom|OnSpriteSpawned
  (params (SpawnedActor :type Actor%20Object%20Reference) (bResult :type Boolean))
  ;[0.2107 0.8438 0.1055] When a sprite is spawned, bind to when it's picked up
  (bind _asnpc_pawn_irwin_simple_sprite (Utilities|Casting|CastToNPC_Pawn_Irwin_Simple_Sprite SpawnedActor))
  (Default|BindEventtoOnSpritePickedUp _asnpc_pawn_irwin_simple_sprite (EventDispatchers|CreateEvent :!event Custom|On%20Sprite%20Picked%20Up)))

(event OnComponentBeginOverlap(DiscoverySphere)
  (params (OverlappedComponent :type Primitive%20Component%20Object%20Reference) (OtherActor :type Actor%20Object%20Reference) (OtherComp :type Primitive%20Component%20Object%20Reference) (OtherBodyIndex :type Integer) (bFromSweep :type Boolean) (SweepResult :type Hit%20Result%20Structure%20%28by%20ref%29))
  ;[0.2107 0.8438 0.1055] Reveal on minimap when closeby
  (bind _asfort_player_pawn (Utilities|Casting|CastToFortPlayerPawn OtherActor)
    (:then
      (CallFunction|RevealMinimapIcon _asfort_player_pawn))
    (:CastFailed
      (bind _asfort_athena_vehicle (Utilities|Casting|CastToFortAthenaVehicle OtherActor))
      (for _array_element (Vehicle|GetAllPassengers _asfort_athena_vehicle)
        (CallFunction|RevealMinimapIcon _array_element)))))

(event Custom|ShrineActivatedNoRep
  (bind _gc_shrineactive_gameplaycuetag (Utilities|Struct|BreakGameplayCueTag (Variables|GetGC_ShrineActive_GameplayCueTag)))
  (bind (_activationdata_shrinestate_9_df2f47954dfaa90600e2d2a2919b7e52 _activationdata_activator_4_9f61edcf4da917a6c98ee3b240dbd41c _activationdata_spritetag_7_9c22aa8b4b16075a3bf364870db2811d) (Utilities|Struct|BreakShrineActivationData (Variables|GetActivationData_ShrineState_9_DF2F47954DFAA90600E2D2A2919B7E52)))
  ;[0.7899 0.1002 0.2833] Shrine SetActivated WITHOUT OnRep -- this is used because some things in the OnRep become relevant on their own OR we only want to run them once w/o caring about relevancy
  (GameplayCue|AddGameplayCueOnActor(Looping) self _gc_shrineactive_gameplaycuetag (Ability|GameplayCue|MakeGameplayCueParameters :NormalizedMagnitude 0.0 :RawMagnitude 0.0 :AggregatedSourceTags (GameplayTags|MakeGameplayTagContainerfromTag _activationdata_spritetag_7_9c22aa8b4b16075a3bf364870db2811d) :Location "0,0,0" :Normal "0,0,0" :GameplayEffectLevel 1 :AbilityLevel 1 :bReplicateLocationWhenUsingMinimalRepProxy false))
  (CallFunction|MarkShrineonCompletion))

(event Custom|SetShrineActivated
  (params (Activated :type Boolean))
  (bind (_activationdata_shrinestate_9_df2f47954dfaa90600e2d2a2919b7e52 _activationdata_activator_4_9f61edcf4da917a6c98ee3b240dbd41c) (Utilities|Struct|BreakShrineActivationData (Variables|GetActivationData_ShrineState_9_DF2F47954DFAA90600E2D2A2919B7E52)))
  (bind (_activationdata_shrinestate_9_df2f47954dfaa90600e2d2a2919b7e52_2 _activationdata_activator_4_9f61edcf4da917a6c98ee3b240dbd41c_2) (Utilities|Struct|BreakShrineActivationData (Variables|GetActivationData_ShrineState_9_DF2F47954DFAA90600E2D2A2919B7E52)))
  ;[0 0.8244 1] Shrine activation
  (if Activated
    (Utilities|FlowControl|DoOnce false
      (:Completed
        (Networking|SwitchHasAuthority
          (:Authority
            (bind _returnvalue (Class|SpriteDiscoverableManager|GetSpriteDiscoverableManager))
            (Class|SpriteDiscoverableManager|UnregisterDiscoverable _returnvalue self) ; Should not be navigated to once activated
            (CallFunction|SpawnSpriteAI)
            (CallFunction|SpawnLoot)
            (Components|Timeline|Stop (Variables|BGA_FirePetal_Shrine_Parent|GetLiftTimeline))
            (GameplayEffects|ApplyGameplayEffectToSelf :self (Ability|GetFortAbilitySystemComponent _activationdata_activator_4_9f61edcf4da917a6c98ee3b240dbd41c) :GameplayEffectClass "/Sprites/Gameplay/Effects/GE_FirePetal_Sprites_PingCooldown.GE_FirePetal_Sprites_PingCooldown_C" :Level 0.0) ; Ensure the activator doesn't get a Shrine ping after picking up the new sprite
            (GameplayCue|ExecuteGameplayCueOnActor(Burst) self (Utilities|Struct|BreakGameplayCueTag (Utilities|Struct|BreakSpriteShrineConversionData (Variables|Default|GetSpriteConversionData))) (Ability|GameplayCue|MakeGameplayCueParameters 0.0)))
          (:Remote
            (for _array_element (Variables|Default|GetMarkedSprites)
              (bind _asnpc_pawn_irwin_simple_sprite (Utilities|Casting|CastToNPC_Pawn_Irwin_Simple_Sprite _array_element))
              (Class|NPCPawnIrwinSimpleSprite|UnmarkSprite _asnpc_pawn_irwin_simple_sprite))))
        (Collision|SetCollisionEnabled (Variables|Default|GetInteractVolume) "NoCollision")
        (Utilities|FlowControl|Sequence
          (:then_0
            (Components|Timeline|Stop (Variables|BGA_FirePetal_Shrine_Parent|GetLiftTimeline))
            (Transformation|SetRelativeLocationAndRotation (Variables|Default|GetMesh) (Math|Vector|MakeVector 0.0 0.0 (Variables|Default|GetLiftDistance)) "0.000000,360.000000,0.000000" false false))
          (:then_1
            (Utilities|FlowControl|Delay 1.0)
            (GameplayEffects|RemoveActiveEffectswithGrantedTags (Ability|GetFortAbilitySystemComponent _activationdata_activator_4_9f61edcf4da917a6c98ee3b240dbd41c_2) "(GameplayTags=((TagName=\"Granted.FirePetal.Shrine.Camera\")))")))))))

(event Custom|StartFallbackTimer
  ;[0 0.8244 1] Shrine activation
  ;;[0.7899 0 0.04794] Just in case we never receive the event, cancel the wait
  (bind _returnvalue (Utilities|Time|SetTimerbyEvent (EventDispatchers|CreateEvent :!event Custom|CancelWaitAndReset) 1e+01))
  (Variables|Default|SetFailsafeTimer _returnvalue))

(event Custom|OnInteractSprite
  (params (InteractingPawn :type Fort%20Pawn%20Object%20Reference))
  ;[0 0.8244 1] Shrine activation
  ;;[0 0.8244 1] RPC interact so both client/server get the shrine activation
  (if (not (Health|IsDBNO InteractingPawn))
    (Activation|ActivateShrinewithSprite InteractingPawn)))

(event Custom|CancelWaitAndReset
  ;[0 0.8244 1] Shrine activation
  ;;[0.7899 0 0.04794] Cancel the event wait and allow interaction again
  (Variables|Interaction|SetbAllowInteract true))

(event Interaction|EventBlueprintOnInteract
  (params (InteractingPawn :type Fort%20Pawn%20Object%20Reference) (InteractionBeingAttempted :type EInteractionBeingAttempted%20Enum :default "FirstInteraction"))
  ;[0 0.8244 1] Shrine activation
  ;;[0 0.8244 1] OnInteract
  (if (GameplayTags|HasMatchingGameplayTag InteractingPawn "(TagName=\"Athena.Player.Sprite.InInventory\")")
    (CallFunction|OnInteractSprite InteractingPawn)
    (else
      (CallFunction|OnInteractRadar InteractingPawn))))

(event Custom|On%20Sprite%20Picked%20Up
  (params (Player%20Controller :type Fort%20Player%20Controller%20Object%20Reference))
  ;[0.1086 0.02458 0.15 0.5]
  (CallFunction|ServeronSpritePickedUp Player%20Controller))

(event Custom|PlaySpriteAnimation
  (bind _budgeteddummymesh (Variables|Default|GetBudgetedDummyMesh))
  (bind (_spriteconversiondata_boontogrant_4_492bc9fd463fbec89033fa9a1da6ceed _spriteconversiondata_loottiertouse_8_4d19ebb04aac218729e5e3ba5a8ff596 _spriteconversiondata_skeletalmeshforanimation_5_97267b434a331e43cdc936b686b61d4e _spriteconversiondata_skeletalmeshanimbp_11_6cd13bb545c51fd7a767edba6c9606e8 _spriteconversiondata_skeletalmeshmontage_14_c107fa3e46cd23ea18c854afbc3d267a) (Utilities|Struct|BreakSpriteShrineConversionData (Variables|GetSpriteConversionData_BoonToGrant_4_492BC9FD463FBEC89033FA9A1DA6CEED)))
  ;[0.7899 0.1002 0.2833] Circling Animation
  (Animation|Montage|PlayMontage _budgeteddummymesh _spriteconversiondata_skeletalmeshmontage_14_c107fa3e46cd23ea18c854afbc3d267a
    (:then
      (Networking|SwitchHasAuthority
        (:Remote
          (Rendering|SetVisibility _budgeteddummymesh true true))))
    (:OnCompleted
      (Rendering|SetVisibility _budgeteddummymesh false true))))

(event Custom|LiftObelisk
  (bind _lifttimeline (Variables|BGA_FirePetal_Shrine_Parent|GetLiftTimeline))
  (bind _returnvalue (Components|Timeline|GetPlaybackPosition _lifttimeline))
  (bind _hf_captureduration (Variables|Hotfixes|GetHF_CaptureDuration))
  (bind _returnvalue_2 (ScalableFloat|GetValueatLevel _hf_captureduration))
  (bind _returnvalue_3 (Math|Float|SafeDivide _returnvalue _returnvalue_2))
  (bind _mesh (Variables|Default|GetMesh))
  (bind (_cachedrelativeloc_x _cachedrelativeloc_y _cachedrelativeloc_z) (Math|Vector|BreakVector (Variables|GetCachedRelativeLoc_X)))
  (bind _activationdata_shrinestate_9_df2f47954dfaa90600e2d2a2919b7e52 (Utilities|Struct|BreakShrineActivationData (Variables|GetActivationData_ShrineState_9_DF2F47954DFAA90600E2D2A2919B7E52)))
  ;[0.7899 0.1002 0.2833] Lift Timeline
  (if (Utilities|Enum|NotEqual(Enum) _activationdata_shrinestate_9_df2f47954dfaa90600e2d2a2919b7e52 "NewEnumerator2")
    (CallFunction|UpdateIdleLoop)
    (Variables|Default|SetCachedRelativeLoc (Variables|/Script/Engine.SceneComponent|GetRelativeLocation (Variables|Default|GetMesh)))
    (Components|Timeline|SetTimelineLength _lifttimeline _returnvalue_2)
    (Timeline|LiftTimeline Play
      (:Update
        (Transformation|SetRelativeLocation _mesh (Math|Vector|MakeVector 0.0 0.0 (Math|Float|Lerp _cachedrelativeloc_z (Variables|Default|GetLiftDistance) _returnvalue_3)) false false)
        (if (Variables|Default|GetbShouldRotate)
          (Transformation|SetRelativeRotation _mesh (Math|Rotator|MakeRotator 0.0 0.0 (Math|Float|Lerp 0.0 1.8e+02 _returnvalue_3)) false false)))
      (:Finished
        (CallFunction|ChestControlMix false))) ; No curves used - we use this to provide a temporary tick func that can change based on hotfix
  ))

(event OnComponentEndOverlap(CameraVolume)
  (params (OverlappedComponent :type Primitive%20Component%20Object%20Reference) (OtherActor :type Actor%20Object%20Reference) (OtherComp :type Primitive%20Component%20Object%20Reference) (OtherBodyIndex :type Integer))
  ;[0.7899 0.1002 0.2833] End the camera effect if the player leaves the area
  (bind _bdidactorhavetags (Class|AthenaQuestFunctionLibrary|AthenaCheckActorHasTags OtherActor "(GameplayTags=((TagName=\"Granted.FirePetal.Shrine.Camera\")))"))
  (if _bdidactorhavetags
    (GameplayEffects|RemoveActiveEffectswithGrantedTags (Ability|GetFortAbilitySystemComponent OtherActor) "(GameplayTags=((TagName=\"Granted.FirePetal.Shrine.Camera\")))")))

(event Custom|UpdateIdleLoop
  ; Update Idle Audio
  (Audio|Components|Audio|FadeOut (Variables|Default|GetAmbientSound) 2.0 0.0 "Linear" :!class /Script/Engine.AudioComponent)
  (Audio|Components|Audio|FadeIn (Variables|Default|GetAmbientSoundActivated) 1.5 1.0 0.0 "Linear" :!class /Script/Engine.AudioComponent)
  (CallFunction|ChestControlMix true))

(event Custom|SpawnAGV
  ; Update Idle Audio
  (Game|SpawnActorfromClass "/Sprites/Assets/Sounds/Music/AGV_Sprites_MusicCBM_Volume.AGV_Sprites_MusicCBM_Volume_C" (Transformation|GetActorTransform) "Undefined" "MultiplyWithRoot" 0 0))

(event Custom|ChestControlMix
  (params (bActivate :type Boolean))
  ; Update Idle Audio
  (Networking|CanExecuteCosmeticEvents
    (:True
      (if bActivate
        (Audio|Modulation|ActivateControlBusMix (Variables|Default|GetChestControlBusMix))
        (else
          (Audio|Modulation|DeactivateControlBusMix (Variables|Default|GetChestControlBusMix)))))))

(fn ActivateShrineWithSprite
  (params (ShrineActivator :type Fort%20Pawn%20Object%20Reference))
  (bind _returnvalue (Inventory|GetInventoryItemwithTag (Fortnite|GetFortPlayerControllerfromActor ShrineActivator) (Variables|Default|GetSpriteProjectileParentTag)))
  (Utilities|IsValid _returnvalue
    (:Is%20Valid
      (Variables|Default|SetCachedSpriteItem _returnvalue)
      (Activation|StartShrineOffering (GameplayTags|GetOwnedGameplayTags (Item|GetItemDefinition _returnvalue) :!class /Script/GameplayTags.BlueprintGameplayTagLibrary) (Variables|GetShrineActivator) true)
      (CallFunction|StartFallbackTimer)
      (Ability|Advanced|GrantandActivateAbilityonSelf (Ability|GetFortAbilitySystemComponent (Variables|GetShrineActivator)) "/FirePetalShrineGameplay/Gameplay/Abilities/GA_FirePetal_Shrine_PlayerActivate.GA_FirePetal_Shrine_PlayerActivate_C" self)
      (return true))
    (:Is%20Not%20Valid
      (Development|PrintString "No sprite detected" true true "(R=0.789931,G=0.000000,B=0.012467,A=1.000000)")
      (return false))))

(fn AddTeamToInteractFilter
  (params (Player :type Actor%20Object%20Reference))
  (Utilities|Array|AddUnique (Variables|Default|GetInteractedTeams) (Team|GetActorTeam Player)))

(fn BlueprintCanInteract
  (params (InteractingPawn :type Fort%20Pawn%20Object%20Reference) (InteractionBeingAttempted :type EInteractionBeingAttempted%20Enum :default "FirstInteraction") (InteractionType :type TInteractionType%20Enum :default "IT_NoInteraction"))
  (bind _activationdata_shrinestate_9_df2f47954dfaa90600e2d2a2919b7e52 (Utilities|Struct|BreakShrineActivationData (Variables|GetActivationData_ShrineState_9_DF2F47954DFAA90600E2D2A2919B7E52)))
  (return (Math|Boolean|ANDBoolean (Utilities|Enum|Equal(Enum) _activationdata_shrinestate_9_df2f47954dfaa90600e2d2a2919b7e52 "NewEnumerator0") (select (Variables|Default|GetAllowMultipleInteracts?) true (not (Utilities|Array|ContainsItem (Variables|Default|GetInteractedTeams) (Team|GetActorTeam InteractingPawn)))) (not (Variables|Default|GetOnInteractCooldown?)) (Variables|Interaction|GetbAllowInteract) (not (Health|IsDBNO InteractingPawn)) (not (GameplayTags|HasMatchingGameplayTag InteractingPawn "(TagName=\"Athena.Player.Sprite.BlockInteractShrine\")")))))

(fn BlueprintGetInteractionString
  (params (InteractingPawn :type Fort%20Pawn%20Object%20Reference) (InteractionBeingAttempted :type EInteractionBeingAttempted%20Enum :default "FirstInteraction"))
  (bind _activationdata_shrinestate_9_df2f47954dfaa90600e2d2a2919b7e52 (Utilities|Struct|BreakShrineActivationData (Variables|GetActivationData_ShrineState_9_DF2F47954DFAA90600E2D2A2919B7E52)))
  (return (Utilities|Select :Index _activationdata_shrinestate_9_df2f47954dfaa90600e2d2a2919b7e52 :NewEnumerator0 (Utilities|String|SelectText "NSLOCTEXT(\"[64742E012ED020D9AA56139A6A5D3309]\", \"07F3251C4E9D9626D8F6C8B93542F8AE\", \"Return sprite to shrine\")" "NSLOCTEXT(\"[64742E012ED020D9AA56139A6A5D3309]\", \"DEDBABEA4B3AF2FA23D6668B23FE307C\", \"Reveal nearby sprites\")" (GameplayTags|HasMatchingGameplayTag InteractingPawn "(TagName=\"Athena.Player.Sprite.InInventory\")")))))

(fn BlueprintGetInteractionTime
  (params (InteractingPawn :type Fort%20Pawn%20Object%20Reference) (InteractionBeingAttempted :type EInteractionBeingAttempted%20Enum :default "FirstInteraction"))
  (return (ScalableFloat|GetValueatLevel (Variables|Hotfixes|GetHF_InteractionTime)) true))

(fn CanBeDiscovered
  (bind _activationdata_shrinestate_9_df2f47954dfaa90600e2d2a2919b7e52 (Utilities|Struct|BreakShrineActivationData (Variables|GetActivationData_ShrineState_9_DF2F47954DFAA90600E2D2A2919B7E52)))
  (return (Utilities|Enum|NotEqual(Enum) _activationdata_shrinestate_9_df2f47954dfaa90600e2d2a2919b7e52 "NewEnumerator2")))

(fn GetDiscoveryData
  (return (Variables|Default|GetSpriteDiscoveryData)))

(fn GetDummySpawnLocation
  (params (Activator :type Fort%20Pawn%20Object%20Reference))
  (return (+ (* (Math|Vector|GetUnitDirection(Vector) (Transformation|GetActorLocation) (Transformation|GetActorLocation Activator)) (Variables|Default|GetDummyForwardOffset)) (Transformation|GetActorLocation)) (Math|Conversions|RotationFromXVector (Math|Vector|RotateVectorAroundAxis (Math|Vector|GetUnitDirection(Vector) (Transformation|GetActorLocation) (Transformation|GetActorLocation Activator)) -9e+01 "0.000000,0.000000,1.000000")) "1.000000,1.000000,1.000000"))

(fn GetRewardSpawnLocation
  (bind (_activationdata_shrinestate_9_df2f47954dfaa90600e2d2a2919b7e52 _activationdata_activator_4_9f61edcf4da917a6c98ee3b240dbd41c) (Utilities|Struct|BreakShrineActivationData (Variables|GetActivationData_ShrineState_9_DF2F47954DFAA90600E2D2A2919B7E52)))
  (return (Utilities|Operators|Add (Transformation|GetWorldLocation (Variables|Default|GetBase)) (* (Math|Vector|GetUnitDirection(Vector) (Transformation|GetWorldLocation (Variables|Default|GetBase)) (Transformation|GetActorLocation _activationdata_activator_4_9f61edcf4da917a6c98ee3b240dbd41c)) (LocalVariables|GetDistance)) "0.000000,0.000000,500.000000") (Math|Conversions|RotationFromXVector (Math|Vector|GetUnitDirection(Vector) (Transformation|GetWorldLocation (Variables|Default|GetBase)) (Transformation|GetActorLocation _activationdata_activator_4_9f61edcf4da917a6c98ee3b240dbd41c)))))

(fn MarkShrineOnCompletion
  (bind (_activationdata_shrinestate_9_df2f47954dfaa90600e2d2a2919b7e52 _activationdata_activator_4_9f61edcf4da917a6c98ee3b240dbd41c) (Utilities|Struct|BreakShrineActivationData (Variables|GetActivationData_ShrineState_9_DF2F47954DFAA90600E2D2A2919B7E52)))
  (if (ScalableFloat|GetValueasBool (Variables|Hotfixes|GetHF_ShouldMarkAfterCapture?))
    (Utilities|FlowControl|DoOnce false
      (:Completed
        (if (> (Math|Vector|Distance(Vector) (Transformation|GetActorLocation _activationdata_activator_4_9f61edcf4da917a6c98ee3b240dbd41c) (Transformation|GetActorLocation)) (ScalableFloat|GetValueatLevel (Variables|Hotfixes|GetHF_MinDistanceToMarkAfterCapture)))
          (MarkerComponent|MarkActoronServer (Actor|GetComponentbyClass (Fortnite|GetFortPlayerControllerfromActor _activationdata_activator_4_9f61edcf4da917a6c98ee3b240dbd41c) "/Script/FortniteGame.AthenaMarkerComponent") self false true))))))

(fn OnRep_ActivationData
  (bind (_spriteconversiondata_boontogrant_4_492bc9fd463fbec89033fa9a1da6ceed _spriteconversiondata_loottiertouse_8_4d19ebb04aac218729e5e3ba5a8ff596 _spriteconversiondata_skeletalmeshforanimation_5_97267b434a331e43cdc936b686b61d4e _spriteconversiondata_skeletalmeshanimbp_11_6cd13bb545c51fd7a767edba6c9606e8 _spriteconversiondata_skeletalmeshmontage_14_c107fa3e46cd23ea18c854afbc3d267a _spriteconversiondata_soundlibrary_22_817e3e6f45376439cd36449df9b5311f) (Utilities|Struct|BreakSpriteShrineConversionData (Variables|GetSpriteConversionData_BoonToGrant_4_492BC9FD463FBEC89033FA9A1DA6CEED)))
  (bind _activationdata_shrinestate_9_df2f47954dfaa90600e2d2a2919b7e52 (Utilities|Struct|BreakShrineActivationData (Variables|GetActivationData_ShrineState_9_DF2F47954DFAA90600E2D2A2919B7E52)))
  (bind _turtleloop (Variables|Default|GetTurtleLoop))
  (CallFunction|RefreshSpriteConversionData)
  (switch Utilities|FlowControl|Switch|SwitchonEFirePetalShrineState _activationdata_shrinestate_9_df2f47954dfaa90600e2d2a2919b7e52
    (:NewEnumerator0
      (CallFunction|SetShrineActivated false))
    (:NewEnumerator3
      (CallFunction|SetMusicCompleteTrigger))
    (:NewEnumerator1
      (CallFunction|LiftObelisk)
      (Components|Activation|Deactivate (Variables|Default|GetNS_FirePetalShrine_LightPillar))
      (CallFunction|UpdateSpriteMesh)
      (CallFunction|PlaySpriteAnimation))
    (:NewEnumerator2
      (CallFunction|SetShrineActivated true)
      (SoundLibrary|RemoveLibrary (Variables|Default|GetFirePetal_SoundLibrary) _spriteconversiondata_soundlibrary_22_817e3e6f45376439cd36449df9b5311f :!class /Script/SoundLibrary.SoundLibraryComponent)
      (bind _musiccontrolleractor (Variables|Default|GetMusicControllerActor)
        (:then
          (Class|BPSpritesVolumeMusicControllerParent|DisableMusic _musiccontrolleractor))
        (:else
          _))
      (if (Variables|Default|GetbIsTurtleAudio)
        (Audio|Components|Audio|FadeIn _turtleloop 3.0 1.0 0.0 "Linear" :!class /Script/Engine.AudioComponent)
        (Audio|AddEventbyComponent (GameInstanceSubsystems|GetMusicEventSubsystem) _turtleloop "(TagName=\"MusicEventSubsystem.Event.POI\")" "(GameplayTags=)" 0 "")))))

(fn PlaySpriteSuccessAnimation)

(fn RefreshSpriteConversionData
  (bind (_activationdata_shrinestate_9_df2f47954dfaa90600e2d2a2919b7e52 _activationdata_activator_4_9f61edcf4da917a6c98ee3b240dbd41c _activationdata_spritetag_7_9c22aa8b4b16075a3bf364870db2811d) (Utilities|Struct|BreakShrineActivationData (Variables|GetActivationData_ShrineState_9_DF2F47954DFAA90600E2D2A2919B7E52)))
  (bind _returnvalue (GameplayTags|GetTagName _activationdata_spritetag_7_9c22aa8b4b16075a3bf364870db2811d))
  (bind _outitem (DataRegistry|FindDataRegistryItem (Utilities|Struct|MakeDataRegistryId (Variables|Default|GetSpriteConversionDataRegistry) _returnvalue))
    (:Found
      (Variables|Default|SetSpriteConversionData _outitem))
    (:NotFound
      (Development|PrintText (Utilities|Text|FormatText "NSLOCTEXT(\"[64742E012ED020D9AA56139A6A5D3309]\", \"FEF097A94698CF428483CB9C03A28104\", \"Failed to find registry conversion data for {0}\")" _returnvalue))))
  (return))

(fn RevealMinimapIcon
  (params (Player :type Fort%20Player%20Pawn%20Object%20Reference))
  (bind _returnvalue (Fortnite|GetFortPlayerControllerfromActor Player))
  (bind _fortminimapcomponent_discoverable (Variables|Default|GetFortMiniMapComponent_Discoverable))
  (Networking|SwitchHasAuthority
    (:Authority
      (Utilities|IsValid _returnvalue
        (:Is%20Valid
          (switch Utilities|FlowControl|Switch|SwitchonEMiniMapComponentDiscoverableVisibility (Class|FortMiniMapComponentDiscoverable|GetVisibilityforPlayer _fortminimapcomponent_discoverable _returnvalue)
            (:NotVisible
              _)
            (:Visible_NotDiscovered
              _))
          (Class|FortMiniMapComponentDiscoverable|SetVisibilityforPlayer _fortminimapcomponent_discoverable _returnvalue "Discovered"))))))

(fn SetMusicCompleteTrigger
  (bind (_spriteconversiondata_boontogrant_4_492bc9fd463fbec89033fa9a1da6ceed _spriteconversiondata_loottiertouse_8_4d19ebb04aac218729e5e3ba5a8ff596 _spriteconversiondata_skeletalmeshforanimation_5_97267b434a331e43cdc936b686b61d4e _spriteconversiondata_skeletalmeshanimbp_11_6cd13bb545c51fd7a767edba6c9606e8 _spriteconversiondata_skeletalmeshmontage_14_c107fa3e46cd23ea18c854afbc3d267a _spriteconversiondata_soundlibrary_22_817e3e6f45376439cd36449df9b5311f) (Utilities|Struct|BreakSpriteShrineConversionData (Variables|GetSpriteConversionData_BoonToGrant_4_492BC9FD463FBEC89033FA9A1DA6CEED)))
  (SoundLibrary|AddLibrary (Variables|Default|GetFirePetal_SoundLibrary) _spriteconversiondata_soundlibrary_22_817e3e6f45376439cd36449df9b5311f :!class /Script/SoundLibrary.SoundLibraryComponent)
  (bind _musiccontrolleractor (Variables|Default|GetMusicControllerActor)
    (:then
      (Class|BPSpritesVolumeMusicControllerParent|TriggerComplete _musiccontrolleractor))
    (:else)))

(fn SpawnLoot
  (bind (_spriteconversiondata_boontogrant_4_492bc9fd463fbec89033fa9a1da6ceed _spriteconversiondata_loottiertouse_8_4d19ebb04aac218729e5e3ba5a8ff596) (Utilities|Struct|BreakSpriteShrineConversionData (Variables|GetSpriteConversionData_BoonToGrant_4_492BC9FD463FBEC89033FA9A1DA6CEED)))
  (Inventory|SpawnPickupInWorldWithLootTier :LootTierName _spriteconversiondata_loottiertouse_8_4d19ebb04aac218729e5e3ba5a8ff596 :Position (CallFunction|GetRewardSpawnLocation) :OverrideMaxStackCount 0 :bToss true :bTossWithVelocity false :SourceType "Tossed" :Source "LootDrop" :LootWorldLevel 0)
  (return))

(fn SpawnMusicControllerActor
  (Variables|Default|GetMusicControllerActor
    (:then)
    (:else
      (bind _returnvalue (Game|SpawnActorfromClass "/Sprites/Assets/Sounds/Music/BP_Sprites_Volume_Music_Shrines.BP_Sprites_Volume_Music_Shrines_C" (Transformation|GetActorTransform) "Undefined" "MultiplyWithRoot" self (Variables|Default|GetMusicControllerRadius) (Variables|Default|GetbIsTurtleAudio) false 0))
      (Variables|Default|SetMusicControllerActor _returnvalue))))

(fn SpawnSpriteAI
  (bind (_location _rotation) (CallFunction|GetRewardSpawnLocation))
  (bind _returnvalue (Math|Transform|MakeTransform _location _rotation))
  (bind (_spriteconversiondata_boontogrant_4_492bc9fd463fbec89033fa9a1da6ceed _spriteconversiondata_loottiertouse_8_4d19ebb04aac218729e5e3ba5a8ff596 _spriteconversiondata_skeletalmeshforanimation_5_97267b434a331e43cdc936b686b61d4e _spriteconversiondata_skeletalmeshanimbp_11_6cd13bb545c51fd7a767edba6c9606e8 _spriteconversiondata_skeletalmeshmontage_14_c107fa3e46cd23ea18c854afbc3d267a _spriteconversiondata_soundlibrary_22_817e3e6f45376439cd36449df9b5311f _spriteconversiondata_spawndefinition_25_2abea5d64143738863e5e68dc0f52eba _spriteconversiondata_lagereventrow_28_37a2813f4061545c29cec0b4df85631a) (Utilities|Struct|BreakSpriteShrineConversionData (Variables|GetSpriteConversionData_BoonToGrant_4_492BC9FD463FBEC89033FA9A1DA6CEED)))
  (bind (_spriteconversiondata_boontogrant_4_492bc9fd463fbec89033fa9a1da6ceed_2 _spriteconversiondata_loottiertouse_8_4d19ebb04aac218729e5e3ba5a8ff596_2 _spriteconversiondata_skeletalmeshforanimation_5_97267b434a331e43cdc936b686b61d4e_2 _spriteconversiondata_skeletalmeshanimbp_11_6cd13bb545c51fd7a767edba6c9606e8_2 _spriteconversiondata_skeletalmeshmontage_14_c107fa3e46cd23ea18c854afbc3d267a_2 _spriteconversiondata_soundlibrary_22_817e3e6f45376439cd36449df9b5311f_2 _spriteconversiondata_spawndefinition_25_2abea5d64143738863e5e68dc0f52eba_2) (Utilities|Struct|BreakSpriteShrineConversionData (Variables|GetSpriteConversionData_BoonToGrant_4_492BC9FD463FBEC89033FA9A1DA6CEED)))
  (bind (_requestid _returnvalue_2) (Class|FortAthenaLivingWorldManager|TrySpawnEventforRegistry (Class|FortAthenaLivingWorldManager|GetLivingWorldManager) (Utilities|Struct|MakeDataRegistryId "SpriteThrownLagerEvents" _spriteconversiondata_lagereventrow_28_37a2813f4061545c29cec0b4df85631a) _returnvalue 0))
  (if (not _returnvalue_2)
    (bind _returnvalue_3 (AI|GetAISpawner))
    (Utilities|IsValid _returnvalue_3
      (:Is%20Valid
        (bind _returnvalue_4 (Class|FortAthenaAISpawnerData|CreateComponentListfromClass _spriteconversiondata_spawndefinition_25_2abea5d64143738863e5e68dc0f52eba_2 self))
        (Class|AthenaAISpawner|RequestSpawn _returnvalue_3 _returnvalue_4 _returnvalue false -1)
        (return)))))

(fn StartShrineCapture
  (bind (_activationdata_shrinestate_9_df2f47954dfaa90600e2d2a2919b7e52 _activationdata_activator_4_9f61edcf4da917a6c98ee3b240dbd41c _activationdata_spritetag_7_9c22aa8b4b16075a3bf364870db2811d) (Utilities|Struct|BreakShrineActivationData (Variables|GetActivationData_ShrineState_9_DF2F47954DFAA90600E2D2A2919B7E52)))
  (Abilities|TryActivateAbilitybyClass (Ability|GetFortAbilitySystemComponent self) (Variables|Default|GetActivationGAClass) true)
  (Variables|SetActivationData_ShrineState_9_DF2F47954DFAA90600E2D2A2919B7E52 (Utilities|Struct|MakeShrineActivationData "NewEnumerator1" _activationdata_activator_4_9f61edcf4da917a6c98ee3b240dbd41c _activationdata_spritetag_7_9c22aa8b4b16075a3bf364870db2811d))
  (Inventory|RemoveFortItemFromPlayer (Fortnite|GetFortPlayerControllerfromActor _activationdata_activator_4_9f61edcf4da917a6c98ee3b240dbd41c) (Variables|Default|GetCachedSpriteItem))
  (return))

(fn StartShrineOffering
  (params (ElementTags :type Gameplay%20Tag%20Container%20Structure) (Activator :type Fort%20Pawn%20Object%20Reference) (ShrineActivated? :type Boolean))
  (bind _sprite_tag (CallFunction|GetSpriteTagfromContainer ElementTags))
  (Utilities|Time|ClearandInvalidateTimerbyHandle (Variables|Default|GetFailsafeTimer))
  (Ability|GameplayEffect|AddLooseGameplayTags self (GameplayTags|MakeGameplayTagContainerfromTag _sprite_tag) true)
  (Variables|SetActivationData_ShrineState_9_DF2F47954DFAA90600E2D2A2919B7E52 (Utilities|Struct|MakeShrineActivationData (select ShrineActivated? "NewEnumerator3" "NewEnumerator0") (Variables|Default|GetActivator) _sprite_tag))
  (return))

(fn UpdateSpriteMesh
  (bind (_activationdata_shrinestate_9_df2f47954dfaa90600e2d2a2919b7e52 _activationdata_activator_4_9f61edcf4da917a6c98ee3b240dbd41c) (Utilities|Struct|BreakShrineActivationData (Variables|GetActivationData_ShrineState_9_DF2F47954DFAA90600E2D2A2919B7E52)))
  (bind (_spriteconversiondata_boontogrant_4_492bc9fd463fbec89033fa9a1da6ceed _spriteconversiondata_loottiertouse_8_4d19ebb04aac218729e5e3ba5a8ff596 _spriteconversiondata_skeletalmeshforanimation_5_97267b434a331e43cdc936b686b61d4e _spriteconversiondata_skeletalmeshanimbp_11_6cd13bb545c51fd7a767edba6c9606e8) (Utilities|Struct|BreakSpriteShrineConversionData (Variables|GetSpriteConversionData_BoonToGrant_4_492BC9FD463FBEC89033FA9A1DA6CEED)))
  (bind _budgeteddummymesh (Variables|Default|GetBudgetedDummyMesh))
  (Components|SkeletalMesh|SetSkeletalMeshAsset _budgeteddummymesh _spriteconversiondata_skeletalmeshforanimation_5_97267b434a331e43cdc936b686b61d4e)
  (Components|SkeletalMesh|SetAnimInstanceClass _budgeteddummymesh _spriteconversiondata_skeletalmeshanimbp_11_6cd13bb545c51fd7a767edba6c9606e8)
  (if (>= (Math|Vector|DotProduct (Transformation|GetRightVector _budgeteddummymesh) (Transformation|GetActorForwardVector _activationdata_activator_4_9f61edcf4da917a6c98ee3b240dbd41c)) 0)
    (Transformation|SetRelativeRotation _budgeteddummymesh "0.000000,-270.000000,0.000000" false false) ; ensure the sprite is always facing the player
  ))

(fn ConstructionScript
  (if (not (ScalableFloat|GetValueasBool (Variables|Hotfixes|GetHF_ShowOnMinimap)))
    (Collision|SetCollisionEnabled (Variables|Default|GetDiscoverySphere) "NoCollision")
    (Class|FortMiniMapComponentDiscoverable|SetDefaultVisibility (Variables|Default|GetFortMiniMapComponent_Discoverable) "NotVisible" false)))
