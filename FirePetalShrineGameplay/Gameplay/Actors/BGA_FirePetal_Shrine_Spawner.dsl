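; Fire Petal Shrine spawner actor that responds to player interaction by requesting an environment query (EQS) for the shrine system.
; On interact, it gets the Athena game state, finds its Fire Petal Shrine game state component and calls DoEQS on it if both are valid.
; The actor holds no spawn logic itself, and its construction script only calls the parent.
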(event Interaction|EventBlueprintOnInteract
  (params (InteractingPawn :type Fort%20Pawn%20Object%20Reference) (InteractionBeingAttempted :type EInteractionBeingAttempted%20Enum :default "FirstInteraction"))
  (bind _returnvalue (Fortnite|Athena|GameState|GetGameStateAthena))
  (bind _returnvalue_2 (Actor|GetComponentbyClass _returnvalue "/FirePetalShrineGameplay/Gameplay/Components/B_FirePetal_ShrineGameStateComponent.B_FirePetal_ShrineGameStateComponent_C"))
  (Utilities|IsValid _returnvalue
    (:Is%20Valid
      (Utilities|IsValid _returnvalue_2
        (:Is%20Valid
          (Class|BFirePetalShrineGameStateComponent|DoEQS _returnvalue_2))))))

(fn ConstructionScript
  (|CallParentFunction))
